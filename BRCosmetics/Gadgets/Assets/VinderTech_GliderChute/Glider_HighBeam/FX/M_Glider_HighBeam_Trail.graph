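# Unlit translucent Niagara ribbon material for the HighBeam glider trail, tinted by particle color through an emissive adjustment function.
# Opacity multiplies a panned primary mask, a tiling smoke noise and a length-wise power falloff with particle alpha.
# Dynamic parameters offset the noise and scale its pan speed.
# Camera-depth and front fades shape the result, with switchable view-angle fade and depth fade, the latter skipped on forward and mobile paths.
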Material /BRCosmetics/Gadgets/Assets/VinderTech_GliderChute/Glider_HighBeam/FX/M_Glider_HighBeam_Trail
graph {
 properties {bUsedWithNiagaraRibbons=true, blendMode=BLEND_Translucent, pixelDepthOffsetMode=PDOM_Legacy, shadingModel=MSM_Unlit, usedShadingModels=Unlit}
 nodes [
  n1 = ParticleColor()
  MaterialFunctionCall(Emissive=@n1.RGB) {materialFunction=&MF_FX_EmissiveAdjustments}
  n2 = MaterialFunctionCall("FadeSensitivity (optional)"=ScalarParameter() {defaultValue=8, parameterName=ViewAngleFadeSensitivity, sliderMax=16}) {materialFunction=&FadeBasedOnViewAngle}
  n3 = Constant() {r=1}
  n4 = ScalarParameter() {defaultValue=1, parameterName=PrimaryMaskTileU}
  n5 = ScalarParameter() {defaultValue=1, parameterName=PrimaryMaskTileV}
  n6 = Add(
   A=Multiply(A=TextureCoordinate(), B=AppendVector(A=@n4, B=@n5))
   B=AppendVector(A=ScalarParameter() {defaultValue=1, parameterName=PrimaryMaskAddU}, B=ScalarParameter() {defaultValue=1, parameterName=PrimaryMaskAddV})
  )
  n7 = StaticSwitchParameter(True=MaterialFunctionCall(XY=@n6) {materialFunction=&Swizzle}.YX, False=@n6) {parameterName=SwizzlePrimaryMaskUVs}
  n8 = AppendVector(A=ScalarParameter() {parameterName=PrimaryMaskPanSpeedU}, B=ScalarParameter() {parameterName=PrimaryMaskPanSpeedV})
  n9 = Multiply(
   A=Add(
    A=TextureCoordinate() {coordinateIndex=1}
    B=DynamicParameter() {defaultValue=<rgba 1 0 1 1>, paramNames [Speed, Trail Noise Random, Param3, Param4]}.Trail Noise Random
   )
   B=AppendVector(A=ScalarParameter() {defaultValue=0.1, parameterName=CloudTileU}, B=ScalarParameter() {defaultValue=1, parameterName=CloudTileV})
  )
  n10 = Multiply(
   A=AppendVector(A=ScalarParameter() {parameterName=TilingNoisePanSpeedU}, B=ScalarParameter() {parameterName=TilingNoisePanSpeedV})
   B=DynamicParameter() {defaultValue=<rgba 1 0 1 1>, paramNames [Speed, Trail Noise Random, Param3, Param4]}.Speed
  )
  n11 = Multiply(
   A=TextureSampleParameter2D(UVs=Panner(Coordinate=@n7, Speed=@n8) {desc=Default Speed is 0}) {parameterName=PrimaryMask, texture=/Game/Global/Textures/Default/Masks_and_Gradients/T_Water_Top_Ramp}.G
   B=TextureSampleParameter2D(UVs=Panner(Coordinate=@n9, Speed=@n10) {desc=Default Speed is 0}) {parameterName=TilingNoise, texture=/Game/Global/Textures/Default/Tileable_Patterns/T_Smoke_Puff_Tile_Up}.R
  )
  n12 = TextureCoordinate()
  n13 = Power(
   Base=OneMinus(None=ComponentMask(None=@n12) {r=true})
   Exp=ScalarParameter() {defaultValue=1, parameterName=SecondaryMaskPower, sliderMax=10}
  ) {constExponent=0.5}
  n14 = Multiply(A=@n1.A, B=Clamp(None=Multiply(A=@n11, B=@n13)))
  n15 = MaterialFunctionCall(Fade Length=ScalarParameter() {defaultValue=150, parameterName=CamDepthFadeLength}) {materialFunction=&CameraDepthFade}
  n16 = ComponentMask(None=TextureCoordinate()) {r=true}
  n17 = OneMinus(None=Power(Base=OneMinus(None=@n16), Exp=ScalarParameter() {defaultValue=5, parameterName=Front Fade Exp, sliderMax=10}) {constExponent=5})
  n18 = Multiply(
   A=Multiply(A=StaticSwitchParameter(True=@n2.Result, False=@n3) {parameterName=UseViewAngleFade}, B=Multiply(A=@n14, B=@n15.Result))
   B=StaticSwitchParameter(True=Clamp(None=@n17), False=Constant() {r=1}) {defaultValue=true, parameterName=IncludeFrontFade}
  )
  n19 = StaticSwitchParameter(True=DepthFade(Opacity=@n18, FadeDistance=ScalarParameter() {parameterName=DepthFadeDistance}), False=@n18) {parameterName=UseDepthFade}
  ShadingPathSwitch(Default=@n19, Deferred=@n19, Forward=@n18, Mobile=@n18)
 ]
}
