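# Niagara system for the HighBeam glider cosmetic, emitting tumbling feather meshes and ribbon trails from the glider's skeletal mesh.
# A short feather burst drifts under drag, curl noise and gravity, shrinking as it ages.
# Bone and socket anim-trail ribbons get sine-swayed velocity, with integrator-driven speed and opacity, front-end overrides and time-of-day color scaling.
# Speedlines, alternate feather and debug emitters are disabled.
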NiagaraSystem /BRCosmetics/Gadgets/Assets/VinderTech_GliderChute/Glider_HighBeam/FX/NS_Glider_HighBeam
dependencies {
 usedDataInterfaces [
  &NiagaraDataInterfaceVector2DCurve
  &NiagaraDataInterfaceCurve
  &NiagaraDataInterfaceColorCurve
  &NiagaraDataInterfaceArrayFloat
  &NiagaraDataInterfaceArrayInt32
  &NiagaraDataInterfaceArrayColor
  &NiagaraDataInterfaceVectorCurve
 ]
 usedDynamicInputs [
  /Niagara/DynamicInputs/Math/Min_Float
  &ClampFloat
  &Subtract_Float
  &Boolean_NotOperation
 ]
 usedModules [
  &SystemState
  &self:SystemSpawnScript.SystemScriptSource.SystemScriptGraph.NiagaraNodeAssignment_0.SetVariables_#1
  &self:SystemSpawnScript.SystemScriptSource.SystemScriptGraph.NiagaraNodeAssignment_6.SetVariables_#2
  &Timeline
  &self:SystemSpawnScript.SystemScriptSource.SystemScriptGraph.NiagaraNodeAssignment_1.SetVariables_#3
  &self:SystemSpawnScript.SystemScriptSource.SystemScriptGraph.NiagaraNodeAssignment_4.SetVariables_#4
  &self:SystemSpawnScript.SystemScriptSource.SystemScriptGraph.NiagaraNodeAssignment_2.SetVariables_#5
  &FN_Integrator
 ]
}
emitters {
 FN_Debug {
  bEnabled=false
  emitterUpdateScript [
   &EmitterState(Life Cycle Mode=::NewEnumerator1, Loop Behavior=::NewEnumerator1, Loop Duration=1, UseLoopDelay=false)
   &SpawnBurst_Instantaneous(Spawn Count=1, Use Spawn Probability=false)
  ]
  particleSpawnScript [/Niagara/Modules/Spawn/Initialization/InitializeParticle(
    Module.Write Color=true
    Module.Write Lifetime=true
    Module.Write Mass=true
    Module.Write Material Random=true
    Module.Write Position=false
    Module.Write Scale=false
    Module.Write SpriteRotation=false
    Module.Write SpriteSize=true
    Position=&SimulationPosition()
    Sprite Size=<100 100>
   )]
  particleUpdateScript [
   &ParticleState(Kill Particles When Lifetime Has Elapsed=false)
   &FN_Timer()
   /SetVariables(Particles.Debug:NiagaraFloat=$User.Variant) {isSetParameters=true}
   &DynamicMaterialParameters(
    Index 0 Param 1=$Particles.Debug
    Module.Param0WriteEnabled=true
    Module.Param1WriteEnabled=true
    Module.Param2WriteEnabled=true
    Module.Param3WriteEnabled=true
   )
   &CameraOffset(Camera Offset Amount=111)
  ]
  rendererClasses [&NiagaraSpriteRendererProperties]
  renderers [{rendererClass=&NiagaraSpriteRendererProperties}]
 }
 Feather {
  emitterUpdateScript [
   &EmitterState(Life Cycle Mode=::NewEnumerator1, Loop Behavior=::NewEnumerator1, Loop Duration=0.4, UseLoopDelay=false)
   &SpawnRate(SpawnRate=5, Use Spawn Probability=false)
   &SpawnBurst_Instantaneous(
    Spawn Count=&RandomRangeInt(Minimum=20, Maximum=30)
    Use Spawn Probability=false
    UseLoopCountLimit=false
   )
  ]
  particleSpawnScript [
   /Niagara/Modules/Spawn/Initialization/V2/InitializeParticle(
    Color=&SelectLinearColorFromArray(
     Array Sampling Mode=::NewEnumerator0
     Direct Array Index=?
     Color Selection Array=DataInterface() {propertyValues {ColorData [<rgba 0.0782 0.099 0.0299 1>, <rgba 0.1198 0.0784 0.0268 1>]}}
     Random Seed=0
    )
    Lifetime=&RandomRangeFloat(Minimum=0.75, Maximum=1.5)
    Mesh Scale=&Multiply_Vector(
     A=$Engine.Owner.Scale
     B=/Niagara/DynamicInputs/CascadeConversion/SplitVector(
      Double Component Vector Channels:ECascadeNiagaraTwoVectorChannels=::NewEnumerator0
      Double Component:NiagaraFloat=&RandomRangeFloat(Minimum=0.3, Maximum=0.5)
      Single Component:NiagaraFloat=&RandomRangeFloat(Minimum=0, Maximum=1)
     )
    )
    Mesh Scale Mode=::NewEnumerator4
    UsePositionOffset=false
    Write Material Random=true
   )
   &ShapeLocation(
    Box Size=<500 50 50>
    MaskSpawnGroup=false
    Offset=<0 0 20>
    Offset Coordinate Space=::Local
    Offset Mode=::NewEnumerator0
    Override Seed=false
    Shape Primitive=::NewEnumerator2
    Surface Only=false
    UseOwnerScale=true
   )
   &AddVelocity(
    Cone Angle=90
    Cone Axis=<0 0 1>
    Module.UseConeFalloff=false
    Module.UseVelDistribution=false
    Override Seed=false
    Rotation Coordinate Space=::Simulation
    Velocity Mode=::NewEnumerator2
    Velocity Speed=&RandomRangeFloat(Minimum=500, Maximum=1000)
   )
   &AddVelocity(
    Rotation Coordinate Space=::Simulation
    Velocity=&Multiply_VectorByFloat(Vector=$Particles.ShapeLocation.ShapeNormal, Float=50)
   ) {enabled=false, module=AddVelocity001}
   &DynamicMaterialParameters(
    Index 0 Param 1=&RandomRangeFloat(Minimum=0, Maximum=0.9)
    Index 0 Param 2=9
    Index 0 Param 4=&RandomRangeFloat(Minimum=0, Maximum=1)
    Module.Param0WriteEnabled=true
    Module.Param1WriteEnabled=true
    Module.Param2WriteEnabled=true
    Module.Param3WriteEnabled=true
   )
   &InitialMeshOrientation(Use Rotation Vector=false)
   /SetVariables(Particles.SpawnRotationVector=&RandomRangeVector(Minimum=<0 0 0>, Maximum=<1 1 1>)) {isSetParameters=true, module=SetVariables_#6}
  ]
  particleUpdateScript [
   &ParticleState()
   &UpdateMeshOrientation(
    Rotation Rate=&FloatFromCurve(
     FloatCurve=DataInterface() {propertyValues {Curve {keys [{interpMode=RCIM_Cubic, tangentMode=RCTM_User, time=0.5, value=0}, {interpMode=RCIM_Cubic, tangentMode=RCTM_User, time=1, value=1}]}}}
     Scale Curve=5
    )
    Rotation Vector=$Particles.SpawnRotationVector
   )
   &ScaleMeshSize(Scale Factor=&VectorFromFloat(Value=&FloatFromCurve(FloatCurve=DataInterface() {propertyValues {Curve {keys [<at 0.5 1>,<at 1 0>]}}})))
   &Drag(Drag=2, Use Linear Drag=true, Use Rotational Drag=false) {enabled=false}
   &VortexForce(
    UseOriginOffset=true
    Vortex Axis=<1 1 1>
    Vortex Axis Coordinate Space:ENiagaraCoordinateSpace=::Local
    Vortex Force Amount=111
    Vortex Origin Offset:Vector3f=&RandomRangeVector(Minimum=<-70 -50 -75>, Maximum=<70 -70 -25>)
    Vortex Origin Offset Coordinate Space:ENiagaraCoordinateSpace=::Local
   ) {enabled=false}
   &InheritSourceMovement(
    Applied Position Scale=&VectorFromFloat(Value=&FloatFromCurve(FloatCurve=DataInterface() {propertyValues {Curve {keys [
          {interpMode=RCIM_Cubic, tangentMode=RCTM_User, time=0, value=0.65}
          {arriveTangent=-1.3555, interpMode=RCIM_Cubic, leaveTangent=-1.3555, tangentMode=RCTM_User, time=1, value=0}
         ]}}}))
    ApplyForce=false
    ApplyPosition=true
    ApplyPositionFalloff=false
   )
   &AerodynamicDrag(Aerodynamic Drag=0.25, Pivot Offset Minimum Length:Vector3f=<10 10 10>, Rotation Strength=3, Use Linear Drag=true, Use Rotational Drag=true)
   &CurlNoiseForce(Noise Strength=500, Pan Noise=false)
   &GravityForce()
   &SolveForcesAndVelocity(Clamp Velocity=false, Limit Acceleration=false)
   &AlignSpriteToMeshOrientation(SetAlignment=true, SetFacing=true)
  ]
  rendererClasses [&NiagaraMeshRendererProperties]
  renderers [{rendererClass=&NiagaraMeshRendererProperties, rendererIndex=0}]
 }
 Feather001 {
  bEnabled=false
  emitterUpdateScript [
   &EmitterState(Life Cycle Mode=::NewEnumerator1, Loop Behavior=::NewEnumerator1, Loop Duration Mode=::NewEnumerator1, UseLoopDelay=false)
   &SpawnRate(SpawnRate=20, Use Spawn Probability=false)
   &SpawnBurst_Instantaneous(
    Spawn Count=&RandomRangeInt(Minimum=4, Maximum=6)
    Use Spawn Probability=false
    UseLoopCountLimit=false
   )
   &SpawnBurst_Instantaneous(
    Spawn Count=&RandomRangeInt(Minimum=1, Maximum=2)
    Spawn Time=0.15
    Use Spawn Probability=false
    UseLoopCountLimit=false
   ) {module=SpawnBurst_Instantaneous001}
   &SpawnBurst_Instantaneous(
    Spawn Count=&RandomRangeInt(Minimum=1, Maximum=2)
    Spawn Time=0.3
    Use Spawn Probability=false
    UseLoopCountLimit=false
   ) {module=SpawnBurst_Instantaneous002}
  ]
  particleSpawnScript [
   /SetVariables(Particles.MeshOrientation=&MakeQuatFromAxisAngle(Axis=&RandomRangeVector(Minimum=<0 0 0>, Maximum=<1 1 1>), AngleInDegrees=90)) {isSetParameters=true}
   /SetVariables(Particles.SpawnRotationVector=&RandomRangeVector(Minimum=<0 0 0>, Maximum=<1 1 1>)) {isSetParameters=true, module=SetVariables_#6}
   &InitializeParticle(
    Lifetime=&RandomRangeFloat(Minimum=1.3, Maximum=2.2)
    Mesh Scale=&SplitVector(
     Double Component Vector Channels:ECascadeNiagaraTwoVectorChannels=::NewEnumerator0
     Double Component:NiagaraFloat=&RandomRangeFloat(Minimum=0.3, Maximum=0.5)
     Single Component:NiagaraFloat=&RandomRangeFloat(Minimum=0, Maximum=1)
    )
    Mesh Scale Mode=::NewEnumerator4
    UsePositionOffset=false
    Write Material Random=true
   )
   &ShapeLocation(
    MaskSpawnGroup=false
    Offset=&RandomRangeVector(Minimum=<-170 -70 -30>, Maximum=<170 -30 -20>)
    Offset Coordinate Space=::Local
    Offset Mode=::NewEnumerator0
    Override Seed=false
    Sphere Radius=5
    UseOwnerScale=true
   )
   &AddVelocity(
    Rotation Coordinate Space=::Simulation
    Velocity=&Multiply_VectorByFloat(Vector=$Particles.ShapeLocation.ShapeVector, Float=&RandomRangeFloat(Minimum=5, Maximum=10))
   ) {module=AddVelocity001}
   &AddVelocity(
    Rotation Coordinate Space=::Simulation
    Velocity=&RandomRangeVector(Minimum=<0 0 -150>, Maximum=<0 0 -800>)
   )
   &DynamicMaterialParameters(
    Index 0 Param 1=&RandomRangeFloat(Minimum=0, Maximum=0.9)
    Index 0 Param 4=&RandomRangeFloat(Minimum=0, Maximum=1)
    Module.Param0WriteEnabled=true
    Module.Param1WriteEnabled=false
    Module.Param2WriteEnabled=false
    Module.Param3WriteEnabled=true
   )
  ]
  particleUpdateScript [
   &ParticleState()
   &UpdateMeshOrientation(Rotation Vector=$Particles.SpawnRotationVector)
   &ScaleMeshSize(Scale Factor=&VectorFromFloat(Value=&FloatFromCurve(FloatCurve=DataInterface() {propertyValues {Curve {keys [<at 0.5 1>,<at 1 0>]}}})))
   &Drag(Drag=4, Use Linear Drag=true, Use Rotational Drag=false)
   &VortexForce(
    UseOriginOffset=true
    Vortex Axis=<1 1 1>
    Vortex Axis Coordinate Space:ENiagaraCoordinateSpace=::Local
    Vortex Force Amount=100
    Vortex Origin Offset:Vector3f=&RandomRangeVector(Minimum=<-70 -50 -75>, Maximum=<70 -70 -25>)
    Vortex Origin Offset Coordinate Space:ENiagaraCoordinateSpace=::Local
   )
   &SolveForcesAndVelocity(Clamp Velocity=false, Limit Acceleration=false)
   &DynamicMaterialParameters(
    Index 0 Param 2=&FloatFromCurve(FloatCurve=DataInterface() {propertyValues {Curve {keys [<at 0.6 1 RCIM_Cubic>,<at 1 0 RCIM_Cubic>]}}})
    Module.Param0WriteEnabled=false
    Module.Param1WriteEnabled=true
    Module.Param2WriteEnabled=true
    Module.Param3WriteEnabled=false
   ) {module=DynamicMaterialParameters001}
   &ScaleColor(
    Scale Alpha=&FloatFromCurve(FloatCurve=DataInterface() {propertyValues {Curve {keys [<at 0 1>,<at 0.6 1>,<at 1 0>]}}})
    Scale RGB=&VectorFromCurve(VectorCurve=DataInterface() {propertyValues {XCurve {keys [<at 0 0.6>,<at 0.25 0.5>]}, YCurve {keys [<at 0 0.6>,<at 0.25 0.5>]}, ZCurve {keys [<at 0 0.6>,<at 0.25 0.5>]}}})
    ScaleA=true
    ScaleRGB=true
   )
  ]
  rendererClasses [&NiagaraMeshRendererProperties]
  renderers [{rendererClass=&NiagaraMeshRendererProperties, rendererIndex=0}]
 }
 Speedlines {
  bEnabled=false
  emitterUpdateScript [
   &EmitterState()
   &SpawnRate(
    SpawnRate=&MakeCustomFloatFromBool(Bool=$User.bIsFrontEndPreview, True Float=50, False Float=90)
    Use Spawn Probability=true
   ) {module=SpawnRate001}
  ]
  particleSpawnScript [
   &InitializeParticle(
    Color=&MakeLinearColorFromVectorAndFloat(
     "Vector (RGB)"=<1 1 1>
     "Float (Alpha)"=&Multiply_Float(
      A=&UniformRangedFloat(Minimum=0.1, Maximum=0.18)
      B=&MakeCustomFloatFromBool(Bool=$User.bIsFrontEndPreview, True Float=0.4, False Float=1)
     )
    )
    Lifetime Max=0.16
    Lifetime Min=0.1
    Lifetime Mode=::NewEnumerator1
    Position Offset=&FN_FrontEndVector3Modifier(Gameplay=<0 0 0>, Front End=<0 50 0>)
    Sprite Size Max=<3 10>
    Sprite Size Min=<2 2>
    Sprite Size Mode=::NewEnumerator2
    UsePositionOffset=true
    Write Material Random=true
   )
   &TorusLocation(
    Align Torus With Axis=true
    Handle Radius=0
    Large Radius:NiagaraFloat=&RandomRangeFloat(Minimum=99, Maximum=111)
    MaskSpawnGroup=false
    Module.Override Randomness=false
    Offset=<0 -180 0>
    Torus Axis=<0 1 0>
   )
   &AddVelocityFromPoint(
    Module.Use Falloff Distance=false
    Origin Offset=<0 -555 0>
    Velocity Strength=&FN_FrontEndFloatModifier(Gameplay=600, Front End=600)
   )
   &AddVelocity(
    Velocity=&RandomRangeVector(Minimum=<0 300 0>, Maximum=<1 700 1>)
    Velocity Speed Scale=&FN_FrontEndFloatModifier(Gameplay=1, Front End=0)
   )
  ]
  particleUpdateScript [
   &SolveForcesAndVelocity(Clamp Velocity=false, Limit Acceleration=false)
   &ParticleState()
   &ScaleSpriteSize(Scale Factor=&Vector2DFromCurve(
     Vector2Curve=DataInterface() {propertyValues {XCurve {keys [<at 0 0.5 RCIM_Cubic>,<at 0.5 1 RCIM_Cubic>]}, YCurve {keys [<at 0 0 RCIM_Cubic>,<at 0.5 1 RCIM_Cubic>]}}}
     ScaleCurve=&Vector2DFromFloat(Value=1)
    ))
   &ScaleSpriteSizeBySpeed(Max Scale Factor=<1 50>, Min Scale Factor=<0.01 0.05>, Velocity Threshold=1)
   &ScaleColor(
    Scale Alpha=&FloatFromCurve(
     FloatCurve=DataInterface() {propertyValues {Curve {keys [
         <at 0 0 RCIM_Cubic>
         <at 0.3759 1.0207>
         {arriveTangent=-2.221, interpMode=RCIM_Cubic, leaveTangent=-2.221, tangentMode=RCTM_User, time=1, value=0}
        ]}}}
     Scale Curve=&MakeCustomFloatFromBool(
      Bool=$User.bIsFrontEndPreview
      True Float=&MakeCustomFloatFromBool(Bool=$User.bIsFullyDeployed, True Float=1, False Float=0)
      False Float=&Lerp_Float(A=0, B=1, Alpha=$User.ForwardDot)
     )
    )
    Scale RGB=&VectorFromFloat(Value=1)
    ScaleA=true
    ScaleRGB=true
   )
  ]
  rendererClasses [&NiagaraSpriteRendererProperties]
  renderers [{rendererClass=&NiagaraSpriteRendererProperties}]
 }
 Trails_Bones {
  emitterSpawnScript [&FN_RibbonIDOffsetUtility()]
  emitterUpdateScript [
   &EmitterState()
   &FN_SpawnAnimTrail(Number of Anim Trails=6, Spawn Interval=0, Spawn Mode:ENiagaraSkeletalSpawning=::NewEnumerator0)
  ]
  particleSpawnScript [
   &FN_InitializeAnimTrailBoneIndecies()
   &FN_InitializeAnimTrailBoneIndecies(AnimTrail Selection=1, First Bone or Socket Index=2) {module=FN_InitializeAnimTrailBoneIndecies001}
   &FN_InitializeAnimTrailBoneIndecies(AnimTrail Selection=2, First Bone or Socket Index=3, Second Bone or Socket Index=2) {module=FN_InitializeAnimTrailBoneIndecies002}
   &FN_InitializeAnimTrailBoneIndecies(AnimTrail Selection=3, First Bone or Socket Index=4, Second Bone or Socket Index=5) {module=FN_InitializeAnimTrailBoneIndecies003}
   &FN_InitializeAnimTrailBoneIndecies(AnimTrail Selection=4, First Bone or Socket Index=6, Second Bone or Socket Index=5) {module=FN_InitializeAnimTrailBoneIndecies004}
   &FN_InitializeAnimTrailBoneIndecies(AnimTrail Selection=5, First Bone or Socket Index=7, Second Bone or Socket Index=6) {module=FN_InitializeAnimTrailBoneIndecies005}
   &FN_InitializeAnimTrail(
    AnimTrailWidth=2
    Color=<rgba 0.25 0.2325 0.2125 1>
    Lifetime=&Multiply_Float(
     A=&SelectFloatFromArray(Direct Array Index=$Particles.SpawnGroup, Float Selection Array=DataInterface() {propertyValues {FloatData [0.3,0.5,0.4,0.3,0.5,0.4]}})
     B=0.5
    )
    Position Offset=<0 0 5>
   )
   &SampleSkeletalMesh(
    AdditionalYawPitchRoll=false
    Bone / Socket Index:NiagaraInt32=&SelectIntFromArray(Direct Array Index=$Particles.SpawnGroup, Int Selection Array=DataInterface() {propertyValues {IntData [0,2,3,4,6,7]}})
    Mesh Sampling Type:ENiagaraSkelSamplingModeFull=::NewEnumerator5
    Skeletal Mesh:NiagaraDataInterfaceSkeletalMesh=$User.SkeletalMesh
    Socket Sampling Mode:ENiagaraSocketSamplingMode=::NewEnumerator5
   )
   &AddVelocity(Velocity=&RotateVectorByQuaternion(
     Vector To Rotate=&Multiply_VectorByFloat(
      Vector=&Multiply_VectorByFloat(
       Vector=&MakeVector(
        X=&Sine(
         Normalized Angle=$Emitter.Age
         Period=&SelectFloatFromArray(Direct Array Index=$Particles.SpawnGroup, Float Selection Array=DataInterface() {propertyValues {FloatData [2,1.5,1.6,3,1,1.2]}})
         Scale=10
         Bias=0
        )
        Y=1500
        Z=&Lerp_Float(A=-50, B=100, Alpha=&Sine(Normalized Angle=$Emitter.Age, Period=2, Scale=0.5, Bias=0.5))
       )
       Float=&MakeCustomFloatFromBool(
        Bool=$User.bIsFrontEnd
        True Float=1
        False Float=&Lerp_Float(A=0.8, B=1, Alpha=$System.FN_Integrator001.NormalizedValue)
       )
      )
      Float=&Lerp_Float(
       A=&SelectFloatFromArray(Direct Array Index=&SpawnGroup(), Float Selection Array=DataInterface() {propertyValues {FloatData [-1,1]}})
       B=1
       Alpha=$System.FN_Integrator001.NormalizedValue
      )
     )
     Quaternion=$Particles.SampleSkeletalMesh.SampledRotation
    )) {enabled=false, module=AddVelocity001}
   &AddVelocity(Velocity=&Multiply_VectorByFloat(
     Vector=&MakeVector(
      X=&Sine(
       Normalized Angle=$Emitter.Age
       Period=&SelectFloatFromArray(Direct Array Index=$Particles.SpawnGroup, Float Selection Array=DataInterface() {propertyValues {FloatData [2,1.5,1.6,3,1,1.2]}})
       Scale=&FN_FrontEndFloatModifier(Gameplay=10, Front End=100)
       Bias=0
      )
      Y=1500
      Z=&Lerp_Float(
       A=&FN_FrontEndFloatModifier(Gameplay=100, Front End=-50)
       B=&FN_FrontEndFloatModifier(Gameplay=200, Front End=100)
       Alpha=&Sine(
        Normalized Angle=$Emitter.Age
        Period=&SelectFloatFromArray(Direct Array Index=$Particles.SpawnGroup, Float Selection Array=DataInterface() {propertyValues {FloatData [2,1.5,1.6,3,1,1.2]}})
        Scale=0.5
        Bias=0.5
       )
      )
     )
     Float=&MakeCustomFloatFromBool(
      Bool=$User.bIsFrontEnd
      True Float=-1
      False Float=&Lerp_Float(A=0.8, B=1.25, Alpha=$System.FN_Integrator001.NormalizedValue)
     )
    ))
   &CameraOffset(Camera Offset Amount=50) {enabled=false}
   &DynamicMaterialParameters(
    Index 0 Param 1=&SelectFloatFromArray(Direct Array Index=$Particles.SpawnGroup, Float Selection Array=DataInterface() {propertyValues {FloatData [1,0.8,1.2,1.3,0.9,0.6]}})
    Module.Param0WriteEnabled=true
    Module.Param1WriteEnabled=false
    Module.Param2WriteEnabled=false
    Module.Param3WriteEnabled=false
   )
  ]
  particleUpdateScript [
   &SampleSkeletalMesh(
    AdditionalYawPitchRoll=false
    Bone / Socket Index:NiagaraInt32=&SelectIntFromArray(Direct Array Index=$Particles.SpawnGroup, Int Selection Array=DataInterface() {propertyValues {IntData [0,2]}})
    Mesh Sampling Type:ENiagaraSkelSamplingModeFull=::NewEnumerator5
    Skeletal Mesh:NiagaraDataInterfaceSkeletalMesh=$User.SkeletalMesh
    Socket Sampling Mode:ENiagaraSocketSamplingMode=::NewEnumerator5
   ) {enabled=false, module=SampleSkeletalMesh001}
   &InheritSourceMovement(
    Applied Position Scale=&MakeVector(
     X=0
     Y=0
     Z=&FloatFromCurve(FloatCurve=DataInterface() {propertyValues {Curve {keys [
          {interpMode=RCIM_Cubic, tangentMode=RCTM_User, time=0.5, value=0.65}
          {arriveTangent=-2.9682, interpMode=RCIM_Cubic, leaveTangent=-2.9682, tangentMode=RCTM_User, time=1, value=0}
         ]}}})
    )
    ApplyForce=false
    ApplyPosition=true
    ApplyPositionFalloff=false
    Source Velocity:Vector3f=$Particles.SampleSkeletalMesh.SampledVelocity
   ) {enabled=false}
   &SolveForcesAndVelocity(Clamp Velocity=false, Limit Acceleration=false)
   &FN_UpdateAnimTrail(SkeletalMesh=$User.SkeletalMesh) {module=FN_UpdateAnimTrail001}
   /SetVariables(Particles.RibbonU0Override=$Particles.NormalizedAge) {isSetParameters=true, module=SetVariables_#7}
   /Niagara/Modules/Update/Color/ScaleColorBySpeed(
    A Scale Max=0.25
    A Scale Min=1
    Max Speed Threshold=1000
    Min Speed Threshold=500
    RGB Scale Min=1
    Source Velocity:Vector3f=$Particles.SampleSkeletalMesh.SampledVelocity
   ) {enabled=false}
   &InterpolateOverTime(
    Initial Color Value:LinearColor=<rgba 1 1 1 1>
    Rate Of Change=6
    Target Color Value:LinearColor=$Transient.ParticleColorScaleFactor
    Variable Type=::NewEnumerator4
   ) {enabled=false}
   &ScaleColor(
    Scale Alpha=&FloatFromCurve(
     FloatCurve=DataInterface() {propertyValues {Curve {keys [
         {tangentMode=RCTM_User, time=0, value=1}
         {arriveTangent=-0.0405, interpMode=RCIM_Cubic, leaveTangent=-0.0405, tangentMode=RCTM_User, time=0.5, value=1}
         {arriveTangent=-3.871, interpMode=RCIM_Cubic, leaveTangent=-3.871, tangentMode=RCTM_User, time=1, value=0}
        ]}}}
     Scale Curve=&Multiply_Float(
      A=$System.Opacity
      B=&FN_FrontEndFloatModifier(
       Gameplay=&RemapRange(Input Value To Remap=$System.FN_Integrator001.NormalizedValue, Input Min=0, Input Max=1, Output Min=0.7, Output Max=0.8)
       Front End=0.7
      )
     )
    )
    Scale RGB=&VectorFromFloat(Value=&FNTimeOfDayScalar(Night Scale=0.7, Morning Scale=1, Evening Scale=1))
    ScaleA=true
    ScaleRGB=true
   )
   &ScaleColor(
    Color Value To Scale=$Particles.Color
    Scale Alpha=&FNTimeOfDayScalar(Night Scale=1, Morning Scale=1.5, Day Scale=2, Evening Scale=1.2)
    ScaleA=true
    ScaleRGB=true
   ) {module=ScaleColor001}
   &ScaleRibbonWidth(Ribbon Width Scale=&FloatFromCurve(FloatCurve=DataInterface() {propertyValues {Curve {keys [{interpMode=RCIM_Cubic, tangentMode=RCTM_User, time=0, value=0.6}, {interpMode=RCIM_Cubic, tangentMode=RCTM_User, time=1, value=1.375}]}}}))
   /SetVariables(Particles.Color=&MultiplyLinearColor(
     Linear Color=$Particles.Color
     Scale Factor=&MakeVector4_FromLinearColor(LinearColor=$Particles.InterpolateOverTime.InterpolateOverTime_LinearColor.Moving Average)
    )) {enabled=false, isSetParameters=true, module=SetVariables_#8}
   /SetVariables(Particles.Velocity=&Multiply_VectorByFloat(
     Vector=$Particles.Initial.Velocity
     Float=&MakeFloatFromLinearColor(
      LinearColor=$Particles.InterpolateOverTime.InterpolateOverTime_LinearColor.Moving Average
      Channel:ENiagaraLinearColor_Channels=::NewEnumerator3
     )
    )) {enabled=false, isSetParameters=true, module=SetVariables_#9}
   /SetVariables(Particles.Vector:Vector3f=$Particles.SampleSkeletalMesh.SampledVelocity) {enabled=false, isSetParameters=true, module=SetVariables_#10}
   /SetVariables(Particles.float:NiagaraFloat=&VectorLength(Vector=$Particles.SampleSkeletalMesh.SampledVelocity)) {enabled=false, isSetParameters=true, module=SetVariables_#11}
  ]
  rendererClasses [&NiagaraRibbonRendererProperties]
  renderers [{rendererClass=&NiagaraRibbonRendererProperties}]
 }
 Trails_Socket {
  emitterSpawnScript [&FN_RibbonIDOffsetUtility()]
  emitterUpdateScript [
   &EmitterState()
   &FN_SpawnAnimTrail(Number of Anim Trails=2, Spawn Interval=0)
  ]
  particleSpawnScript [
   &FN_InitializeAnimTrailBoneIndecies(First Bone or Socket Index=1, Second Bone or Socket Index=0)
   &FN_InitializeAnimTrailBoneIndecies(AnimTrail Selection=1, First Bone or Socket Index=3, Second Bone or Socket Index=2) {module=FN_InitializeAnimTrailBoneIndecies001}
   &FN_InitializeAnimTrail(
    AnimTrailWidth=&FN_FrontEndFloatModifier(Gameplay=1.5, Front End=1.2)
    Color=<rgba 0.25 0.2325 0.2125 1.5>
    Lifetime=&Multiply_Float(A=0.5, B=0.5)
    Position Offset=<0 0 5>
   )
   &SampleSkeletalMesh(
    AdditionalYawPitchRoll=false
    Bone / Socket Index:NiagaraInt32=&SelectIntFromArray(Direct Array Index=$Particles.SpawnGroup, Int Selection Array=DataInterface() {propertyValues {IntData [0,2]}})
    Mesh Sampling Type:ENiagaraSkelSamplingModeFull=::NewEnumerator5
    Skeletal Mesh:NiagaraDataInterfaceSkeletalMesh=$User.SkeletalMesh
    Socket Sampling Mode:ENiagaraSocketSamplingMode=::NewEnumerator5
   )
   &AddVelocity(Velocity=&RotateVectorByQuaternion(
     Vector To Rotate=&Multiply_VectorByFloat(
      Vector=&Multiply_VectorByFloat(
       Vector=&MakeVector(
        X=&Sine(
         Normalized Angle=$Emitter.Age
         Period=&SelectFloatFromArray(Direct Array Index=$Particles.SpawnGroup, Float Selection Array=DataInterface() {propertyValues {FloatData [2,1.5,1.6,3,1,1.2]}})
         Scale=10
         Bias=0
        )
        Y=1500
        Z=&Lerp_Float(A=-50, B=100, Alpha=&Sine(Normalized Angle=$Emitter.Age, Period=2, Scale=0.5, Bias=0.5))
       )
       Float=&MakeCustomFloatFromBool(
        Bool=$User.bIsFrontEnd
        True Float=1
        False Float=&Lerp_Float(A=0.8, B=1, Alpha=$System.FN_Integrator001.NormalizedValue)
       )
      )
      Float=&Lerp_Float(
       A=&SelectFloatFromArray(Direct Array Index=&SpawnGroup(), Float Selection Array=DataInterface() {propertyValues {FloatData [-1,1]}})
       B=1
       Alpha=$System.FN_Integrator001.NormalizedValue
      )
     )
     Quaternion=$Particles.SampleSkeletalMesh.SampledRotation
    )) {enabled=false}
   &AddVelocity(Velocity=&Multiply_VectorByFloat(
     Vector=&MakeVector(
      X=&Sine(
       Normalized Angle=$Emitter.Age
       Period=&SelectFloatFromArray(Direct Array Index=$Particles.SpawnGroup, Float Selection Array=DataInterface() {propertyValues {FloatData [2,1.5,1.6,3,1,1.2]}})
       Scale=&FN_FrontEndFloatModifier(Gameplay=10, Front End=10)
       Bias=0
      )
      Y=1500
      Z=&Lerp_Float(
       A=&FN_FrontEndFloatModifier(Gameplay=100, Front End=-50)
       B=&FN_FrontEndFloatModifier(Gameplay=200, Front End=100)
       Alpha=&Sine(
        Normalized Angle=$Emitter.Age
        Period=&SelectFloatFromArray(Direct Array Index=$Particles.SpawnGroup, Float Selection Array=DataInterface() {propertyValues {FloatData [2,1.5,1.6,3,1,1.2]}})
        Scale=0.5
        Bias=0.5
       )
      )
     )
     Float=&MakeCustomFloatFromBool(
      Bool=$User.bIsFrontEnd
      True Float=-1
      False Float=&Lerp_Float(A=0.8, B=1.25, Alpha=$System.FN_Integrator001.NormalizedValue)
     )
    )) {module=AddVelocity001}
   &CameraOffset(Camera Offset Amount=50) {enabled=false}
   &DynamicMaterialParameters(
    Index 0 Param 1=1
    Module.Param0WriteEnabled=true
    Module.Param1WriteEnabled=false
    Module.Param2WriteEnabled=false
    Module.Param3WriteEnabled=false
   )
  ]
  particleUpdateScript [
   &SampleSkeletalMesh(
    AdditionalYawPitchRoll=false
    Bone / Socket Index:NiagaraInt32=&SelectIntFromArray(Direct Array Index=$Particles.SpawnGroup, Int Selection Array=DataInterface() {propertyValues {IntData [0,2]}})
    Mesh Sampling Type:ENiagaraSkelSamplingModeFull=::NewEnumerator5
    Skeletal Mesh:NiagaraDataInterfaceSkeletalMesh=$User.SkeletalMesh
    Socket Sampling Mode:ENiagaraSocketSamplingMode=::NewEnumerator5
   ) {module=SampleSkeletalMesh001}
   &InheritSourceMovement(
    Applied Position Scale=&MakeVector(
     X=0
     Y=0
     Z=&FloatFromCurve(FloatCurve=DataInterface() {propertyValues {Curve {keys [
          {arriveTangent=-2.6461, interpMode=RCIM_Cubic, leaveTangent=-2.6461, tangentMode=RCTM_User, time=0, value=1}
          {arriveTangent=-0.1166, interpMode=RCIM_Cubic, leaveTangent=-0.1166, tangentMode=RCTM_User, time=1, value=0}
         ]}}})
    )
    ApplyForce=false
    ApplyPosition=true
    ApplyPositionFalloff=false
    Source Velocity:Vector3f=$Particles.SampleSkeletalMesh001.SampledVelocity
   )
   &SolveForcesAndVelocity(Clamp Velocity=false, Limit Acceleration=false)
   &FN_UpdateAnimTrail(SkeletalMesh=$User.SkeletalMesh) {module=FN_UpdateAnimTrail001}
   /SetVariables(Particles.RibbonU0Override=$Particles.NormalizedAge) {isSetParameters=true, module=SetVariables_#7}
   &ScaleColorBySpeed(
    A Scale Max=0.25
    A Scale Min=1
    Max Speed Threshold=1000
    Min Speed Threshold=500
    RGB Scale Min=1
    Source Velocity:Vector3f=$Particles.SampleSkeletalMesh.SampledVelocity
   ) {enabled=false}
   &InterpolateOverTime(
    Initial Color Value:LinearColor=<rgba 1 1 1 1>
    Rate Of Change=6
    Target Color Value:LinearColor=$Transient.ParticleColorScaleFactor
    Variable Type=::NewEnumerator4
   ) {enabled=false}
   &ScaleColor(
    Scale Alpha=&FloatFromCurve(
     FloatCurve=DataInterface() {propertyValues {Curve {keys [
         {tangentMode=RCTM_User, time=0, value=1}
         {arriveTangent=-0.0405, interpMode=RCIM_Cubic, leaveTangent=-0.0405, tangentMode=RCTM_User, time=0.5, value=1}
         {arriveTangent=-3.871, interpMode=RCIM_Cubic, leaveTangent=-3.871, tangentMode=RCTM_User, time=1, value=0}
        ]}}}
     Scale Curve=&Multiply_Float(
      A=$System.Opacity
      B=&FN_FrontEndFloatModifier(
       Gameplay=&RemapRange(Input Value To Remap=$System.FN_Integrator001.NormalizedValue, Input Min=0, Input Max=1, Output Min=0.7, Output Max=0.8)
       Front End=0.7
      )
     )
    )
    Scale RGB=&VectorFromFloat(Value=&FNTimeOfDayScalar(Night Scale=0.7, Morning Scale=1, Evening Scale=1))
    ScaleA=true
    ScaleRGB=true
   )
   &ScaleColor(
    Color Value To Scale=$Particles.Color
    Scale Alpha=&FNTimeOfDayScalar(Night Scale=1, Morning Scale=1.5, Day Scale=2, Evening Scale=1.2)
    ScaleA=true
    ScaleRGB=true
   ) {module=ScaleColor001}
   &ScaleRibbonWidth(Ribbon Width Scale=&FloatFromCurve(FloatCurve=DataInterface() {propertyValues {Curve {keys [{interpMode=RCIM_Cubic, tangentMode=RCTM_User, time=0, value=1}, {interpMode=RCIM_Cubic, tangentMode=RCTM_User, time=1, value=1.375}]}}}))
   /SetVariables(Particles.Color=&MultiplyLinearColor(
     Linear Color=$Particles.Color
     Scale Factor=&MakeVector4_FromLinearColor(LinearColor=$Particles.InterpolateOverTime.InterpolateOverTime_LinearColor.Moving Average)
    )) {enabled=false, isSetParameters=true, module=SetVariables_#8}
   /SetVariables(Particles.Velocity=&Multiply_VectorByFloat(
     Vector=$Particles.Initial.Velocity
     Float=&MakeFloatFromLinearColor(
      LinearColor=$Particles.InterpolateOverTime.InterpolateOverTime_LinearColor.Moving Average
      Channel:ENiagaraLinearColor_Channels=::NewEnumerator3
     )
    )) {enabled=false, isSetParameters=true, module=SetVariables_#9}
   /SetVariables(Particles.Vector:Vector3f=$Particles.SampleSkeletalMesh.SampledVelocity) {enabled=false, isSetParameters=true, module=SetVariables_#10}
   /SetVariables(Particles.float:NiagaraFloat=&VectorLength(Vector=$Particles.SampleSkeletalMesh.SampledVelocity)) {enabled=false, isSetParameters=true, module=SetVariables_#11}
  ]
  rendererClasses [&NiagaraRibbonRendererProperties]
  renderers [{rendererClass=&NiagaraRibbonRendererProperties}]
 }
}
summary {
 systemName=NS_Glider_HighBeam
 emitters [
  {bEnabled=false, emitterName=FN_Debug, rendererClasses [&NiagaraSpriteRendererProperties]}
  {bEnabled=false, emitterName=Speedlines, rendererClasses [&NiagaraSpriteRendererProperties]}
  {emitterName=Trails_Bones, rendererClasses [&NiagaraRibbonRendererProperties]}
  {emitterName=Feather, rendererClasses [&NiagaraMeshRendererProperties]}
  {bEnabled=false, emitterName=Feather001, rendererClasses [&NiagaraMeshRendererProperties]}
  {emitterName=Trails_Socket, rendererClasses [&NiagaraRibbonRendererProperties]}
 ]
 userVariables [
  User.bIsBigFlapping {type=NiagaraBool}
  User.bIsFrontEnd {type=NiagaraBool}
  User.bIsFrontEndPreview {type=NiagaraBool}
  User.bIsFullyDeployed {type=NiagaraBool}
  User.blsFlapping {type=NiagaraBool}
  User.FlapCurve {type=NiagaraFloat, defaultValue=0, description=""}
  User.ForwardDot {type=NiagaraFloat, defaultValue=1, description=""}
  User.New Linear Color {type=LinearColor, description="", defaultValue {struct=LinearColor, value=<rgba 0.0052 0.0021 0.0025 10>}}
  User.RibbonIDOffset {type=NiagaraInt32}
  User.RightDot {type=NiagaraFloat, defaultValue=0, description=""}
  User.SkeletalMesh {
   type=NiagaraDataInterfaceSkeletalMesh
   defaultValue {
    struct=NiagaraExt_VariableValue_DataInterface
    value {dataInterface=&self:NiagaraDataInterfaceSkeletalMesh_0}
   }
  }
  User.Variant {type=NiagaraFloat, defaultValue=0, description=""}
  User.Variant_Color {type=LinearColor, description="", defaultValue {struct=LinearColor, value=<rgba 1 1 1 1>}}
 ]
}
